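# Sound cue for the default bow that, going by its name, confirms the bow has reached full draw tension.
# A Mixer layers two confirm waves at equal volume, so both play together rather than as random variations.
# The cue runs at reduced volume in the weapon foley sound class, with weapon reload attenuation and both global and local bow concurrency rules.
# Its debug flag is left on.
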SoundCue /PrimalGameplay/Sounds/Weapons/Bows/Default/Bow_Default_FullTension_Confirm_Cue
root=Mixer_0
nodes [
 {id=Mixer_0, children [WavePlayer_1,WavePlayer_2], properties {inputVolume [1,1]}}
 {id=WavePlayer_1, properties {soundWaveAssetPtr=&/Bow_Default_FullTension_Confirm_01}}
 {id=WavePlayer_2, properties {soundWaveAssetPtr=&/Bow_Default_FullTension_Confirm_02}}
]
properties {
 attenuationSettings=&Fort_Weapon_Reload_Att
 bDebug=true
 maxDistance=2250
 soundClassObject=&Fort_WeaponFoley
 volumeMultiplier=0.451
 concurrencySet [/Game/Sounds/Concurrency_Rules/Bow_Global_Conc, /Game/Sounds/Concurrency_Rules/Bow_Local_Conc]
}
