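# Looping in-air sound cue for default bow arrow projectiles, routed to the ranged weapon projectile effects sound class.
# After a short fixed delay, a looping wave plays through an enveloper that fades volume in and shapes pitch over time, with a modulator randomizing pitch per play.
# An arrow-specific attenuation asset spatializes it, and global and local bow concurrency rules limit simultaneous instances.
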SoundCue /PrimalGameplay/Sounds/Weapons/Bows/Default/Bow_Default_Prj_InAir_Loop_Cue
root=Delay_0
nodes [
 {id=Delay_0, children [Modulator_0], properties {delayMax=0.15, delayMin=0.15}}
 {
  id=Enveloper_0
  children [WavePlayer_0]
  properties {
   pitchCurve {editorCurveData {keys [
      {arriveTangent=-17.2538, interpMode=RCIM_Cubic, time=0}
      {time=0.0009}
      {arriveTangent=0.1496, interpMode=RCIM_Cubic, leaveTangent=0.1496, tangentMode=RCTM_User, time=0.0616, value=1.0113}
      {time=2}
     ]}}
   volumeCurve {editorCurveData {keys [{arriveTangent=-0.1648, interpMode=RCIM_Cubic, leaveTangent=-0.1648, tangentMode=RCTM_User, time=0, value=0}, <at 1 1>]}}
  }
 }
 {id=Modulator_0, children [Enveloper_0], properties {pitchMax=1.2, pitchMin=0.9, volumeMin=1}}
 {id=WavePlayer_0, properties {bLooping=true, soundWaveAssetPtr=&/Bow_Default_Prj_InAir_Loop_02}}
]
properties {
 attenuationSettings=/Game/Sounds/Attenuation/Fort_Weapon_Projectile_Arrow_InAirLoop_Att
 maxDistance=16800
 pitchMultiplier=1.2
 soundClassObject=&Fort_WeaponRanged_Projectile_Effects
 volumeMultiplier=0.6
 attenuationOverrides {falloffDistance=5000, reverbWetLevelMax=0.7, reverbWetLevelMin=0.2, attenuationShapeExtents {x=1000}}
 concurrencySet [/Game/Sounds/Concurrency_Rules/Bow_Global_Conc, /Game/Sounds/Concurrency_Rules/Bow_Local_Conc]
}
